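# Material function that adds a thin-film iridescence effect to ranked cosmetic materials, toggled by a static switch and masked by a selectable alpha channel.
# A Fresnel term, whose normal can blend toward a pivot-centred spherical normal, is warped to sample a spectrum texture.
# The film, with optional hue, desaturation and exponent adjustment, adds onto replaceable base color, weighted by an exponent and roughness influence.
# Roughness also gets a bias.
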MaterialFunction /Game/Global/Materials/Cosmetics/Ranked/MF_RankedThinFilm
graph {nodes [
  n1 = SetMaterialAttributes!(
   MaterialAttributes=NamedRerouteUsage()
   BaseColor=NamedRerouteUsage()
   Roughness=Add(A=NamedRerouteUsage(), B=ScalarParameter() {group=06_Thin Film, parameterName=Roughness_Bias, sliderMax=1, sliderMin=-1})
  ) {attributeSetTypes [#1, #2]}
  n2 = NamedRerouteDeclaration(None=ChannelMaskParameter(None=FunctionInput() {bUsePreviewValueAsDefault=true, inputName=Alpha, sortPriority=3}) {group=05_Tech Art Masks, parameterName=ThinFilmMaskChannel}.R) {name=ThinFilmMask, nodeColor=<rgba 1 0.3516 0 1>}
  n3 = NamedRerouteUsage()
  FunctionOutput(None=StaticSwitchParameter(True=BlendMaterialAttributes(A=@n3, B=@n1, Alpha=@n2) {vertexAttributeBlendType=UseA}, False=@n3) {group=04_Tech Art, parameterName=UseRankedThinFilm})
  n4 = GetMaterialAttributes(None=FunctionInput() {inputName=Material Attributes, inputType=FunctionInput_MaterialAttributes, sortPriority=1}) {attributeGetTypes [#2, #1]}
  NamedRerouteDeclaration(None=@n4.MaterialAttributes) {name=Material Attributes, nodeColor=<rgba 0.0781 0 1 1>}
  n5 = ScalarParameter() {defaultValue=1, group=06_Thin Film, parameterName=ThinFilmScale, sortPriority=5}
  n6 = Power(Base=NamedRerouteUsage(), Exp=ScalarParameter() {defaultValue=0.5, group=06_Thin Film, parameterName=ThinFilmWarp, sortPriority=4})
  n7 = TextureSampleParameter2D(UVs=Add(A=Multiply(A=@n5, B=@n6), B=ScalarParameter() {group=06_Thin Film, parameterName=ThinFilmOffset})) {
   group=06_Thin Film
   parameterName=Thin Film Texture
   sortPriority=1
   texture=/Game/Global/Textures/Default/Thinfilm/T_ThinFilm_Spectrum_COLOR
  }
  NamedRerouteDeclaration(None=Multiply(A=@n7.RGB, B=ScalarParameter() {defaultValue=0.5, group=06_Thin Film, parameterName=ThinFilmIntensity, sortPriority=2})) {name=BaseThinFilm, nodeColor=<rgba 1 0.8672 0 1>}
  n8 = ScalarParameter() {group=06_Thin Film, parameterName=ColorAdjustHueShift, sortPriority=101}
  n9 = Desaturation(None=NamedRerouteUsage(), Fraction=ScalarParameter() {group=06_Thin Film, parameterName=ColorAdjustDesaturate, sortPriority=102})
  NamedRerouteDeclaration(None=Power(
    Base=MaterialFunctionCall(Hue Shift Percentage=@n8, Texture=@n9) {materialFunction=&HueShift}.Result
    Exp=ScalarParameter() {defaultValue=1, group=06_Thin Film, parameterName=ColorAdjustExponent, sortPriority=103}
   )) {name=ColorAdjustmentThinFilm, nodeColor=<rgba 0.5469 0 1 1>}
  n10 = VectorParameter() {
   group=06_Thin Film
   parameterName=BaseColor_Replacement
   channelNames {a=Replace Intensity, b=B, g=G, r=R}
   defaultValue=<rgba 0.03 0.03 0.03 0>
  }
  NamedRerouteDeclaration(None=LinearInterpolate(A=@n4.BaseColor, B=LinearInterpolate(A=@n4.BaseColor, B=@n10, Alpha=@n10), Alpha=NamedRerouteUsage())) {name=IN_BaseColor, nodeColor=<rgba 0.7344 1 0 1>}
  NamedRerouteDeclaration(None=@n4.Roughness) {name=IN_Roughness, nodeColor=<rgba 0.3359 1 0 1>}
  n11 = PixelNormalWS()
  n12 = VectorParameter() {group=06_Thin Film, parameterName=SphericalNormalPosOffset, defaultValue=<rgba 0 0 0 1>}
  n13 = MaterialFunctionCall() {materialFunction=&ObjectPivotPoint}
  n14 = Normalize(VectorInput=Subtract(A=WorldPosition().XYZ, B=Add(A=@n12.RGB, B=@n13.Object Pivot Location)))
  n15 = ScalarParameter() {group=06_Thin Film, parameterName=SphericalNormal_Blend_Amt, sliderMax=1}
  NamedRerouteDeclaration(None=Fresnel(Normal=LinearInterpolate(A=@n11, B=@n14, Alpha=@n15)) {baseReflectFraction=0, exponent=1}) {name=Fresnel, nodeColor=<rgba 1 0 0.2578 1>}
  n16 = NamedRerouteUsage()
  n17 = StaticSwitchParameter(True=NamedRerouteUsage(), False=NamedRerouteUsage()) {group=06_Thin Film, parameterName=UseColorAdjustment, sortPriority=100}
  n18 = Power(Base=NamedRerouteUsage(), Exp=ScalarParameter() {defaultValue=2, group=06_Thin Film, parameterName=ThinFilmExponent})
  n19 = NamedRerouteUsage()
  n20 = LinearInterpolate(
   A=@n18
   B=Multiply(A=LinearInterpolate(Alpha=@n19) {constA=1, constB=0.01}, B=@n18)
   Alpha=ScalarParameter() {group=06_Thin Film, parameterName=RoughnessInfluence}
  )
  NamedRerouteDeclaration(None=LinearInterpolate(A=NamedRerouteUsage(), B=Add(A=@n16, B=@n17), Alpha=Saturate(None=@n20))) {name=OUT_BaseColor, nodeColor=<rgba 0 0.5781 1 1>}
 ]}
